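# Level Sequence for PistolCanary mnm1000 that animates four NPC aim actors and toggles the tentacle boss's visibility.
# Each aim actor has an attach section, with its parent not shown, and location keys that move it from a large offset to zero shortly after attaching.
# The boss visibility track turns on when the attach sections begin and off as they end.
# The sequence has no camera or master tracks.
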LevelSequence /PistolCanary/Sequence/mnm/mnm1000/PistolCanary_mnm1000_anm
bindings [
 BGA_PistolCanary_NPCAimActor2 {
  id=#1
  possessed_class=BGA_PistolCanary_NPCAimActor_C
  tracks [
   {
    type=MovieScene3DTransformTrack
    sections [{
      type=MovieScene3DTransformSection
      channels [
       Location.X {type=MovieSceneScriptingDoubleChannel, default=12947.3599, keys [[3680,-7865.9438],[3850,0]]}
       Location.Y {type=MovieSceneScriptingDoubleChannel, default=78237.5094, keys [[3680,2702.7322],[3850,0]]}
       Location.Z {type=MovieSceneScriptingDoubleChannel, default=3431.0912, keys [[3680,-9153.9639],[3850,0]]}
       Rotation.X {type=MovieSceneScriptingDoubleChannel, default=0, keys [[3680,0],[3850,0]]}
       Rotation.Y {type=MovieSceneScriptingDoubleChannel, default=0, keys [[3680,0],[3850,0]]}
       Rotation.Z {type=MovieSceneScriptingDoubleChannel, default=45, keys [[3680,0],[3850,0]]}
       Scale.X {type=MovieSceneScriptingDoubleChannel, default=1, const [1,2]}
       Scale.Y {type=MovieSceneScriptingDoubleChannel, default=1, const [1,2]}
       Scale.Z {type=MovieSceneScriptingDoubleChannel, default=1, const [1,2]}
      ]
     }]
   }
   {type=MovieScene3DAttachTrack, sections [{type=MovieScene3DAttachSection, end_frame=15518, start_frame=3600}]}
  ]
 }
 BGA_PistolCanary_NPCAimActor3 {
  id=#2
  possessed_class=BGA_PistolCanary_NPCAimActor_C
  tracks [
   {
    type=MovieScene3DTransformTrack
    sections [{
      type=MovieScene3DTransformSection
      channels [
       Location.X {type=MovieSceneScriptingDoubleChannel, default=13565.4058, keys [[3680,-13630.4141],[3850,0]]}
       Location.Y {type=MovieSceneScriptingDoubleChannel, default=78080.3274, keys [[3680,2420.3945],[3850,0]]}
       Location.Z {type=MovieSceneScriptingDoubleChannel, default=3320.8515, keys [[3680,-3213.9116],[3850,0]]}
       Rotation.X {type=MovieSceneScriptingDoubleChannel, default=0, keys [[3680,0],[3850,0]]}
       Rotation.Y {type=MovieSceneScriptingDoubleChannel, default=0, keys [[3680,0],[3850,0]]}
       Rotation.Z {type=MovieSceneScriptingDoubleChannel, default=0, keys [[3680,0],[3850,0]]}
       Scale.X {type=MovieSceneScriptingDoubleChannel, default=1, const [1,2]}
       Scale.Y {type=MovieSceneScriptingDoubleChannel, default=1, const [1,2]}
       Scale.Z {type=MovieSceneScriptingDoubleChannel, default=1, const [1,2]}
      ]
     }]
   }
   {type=MovieScene3DAttachTrack, sections [{type=MovieScene3DAttachSection, end_frame=15518, start_frame=3600}]}
  ]
 }
 TentacleBoss {
  id=#3
  possessed_class=PistolCanary_TentacleBoss_C
  tracks [{
    type=MovieSceneVisibilityTrack
    sections [{type=MovieSceneVisibilitySection, channels [None {type=MovieSceneScriptingBoolChannel, default=true, keys [[3600,true],[15517,false]]}]}]
   }]
 }
 BGA_PistolCanary_NPCAimActor {
  id=#4
  possessed_class=BGA_PistolCanary_NPCAimActor_C
  tracks [
   {
    type=MovieScene3DTransformTrack
    sections [{
      type=MovieScene3DTransformSection
      channels [
       Location.X {type=MovieSceneScriptingDoubleChannel, default=13932.0287, keys [[3680,223.46],[3850,0]]}
       Location.Y {type=MovieSceneScriptingDoubleChannel, default=78329.4384, keys [[3680,3070.2905],[3850,0]]}
       Location.Z {type=MovieSceneScriptingDoubleChannel, default=3544.3281, keys [[3680,-11335.2871],[3850,0]]}
       Rotation.X {type=MovieSceneScriptingDoubleChannel, default=0, const [0,2]}
       Rotation.Y {type=MovieSceneScriptingDoubleChannel, default=0, const [0,2]}
       Rotation.Z {type=MovieSceneScriptingDoubleChannel, default=0, const [0,2]}
       Scale.X {type=MovieSceneScriptingDoubleChannel, default=1, const [1,2]}
       Scale.Y {type=MovieSceneScriptingDoubleChannel, default=1, const [1,2]}
       Scale.Z {type=MovieSceneScriptingDoubleChannel, default=1, const [1,2]}
      ]
     }]
   }
   {type=MovieScene3DAttachTrack, sections [{type=MovieScene3DAttachSection, end_frame=15518, start_frame=3600}]}
  ]
 }
 BGA_PistolCanary_NPCAimActor4 {
  id=#5
  possessed_class=BGA_PistolCanary_NPCAimActor_C
  tracks [
   {
    type=MovieScene3DTransformTrack
    sections [{
      type=MovieScene3DTransformSection
      channels [
       Location.X {type=MovieSceneScriptingDoubleChannel, default=14563.2732, keys [[3680,-3688.1501],[3850,0]]}
       Location.Y {type=MovieSceneScriptingDoubleChannel, default=78205.4331, keys [[3680,4034.7876],[3850,0]]}
       Location.Z {type=MovieSceneScriptingDoubleChannel, default=3357.4295, keys [[3680,-16538.8848],[3850,0]]}
       Rotation.X {type=MovieSceneScriptingDoubleChannel, default=0, keys [[3680,0],[3850,0]]}
       Rotation.Y {type=MovieSceneScriptingDoubleChannel, default=0, keys [[3680,0],[3850,0]]}
       Rotation.Z {type=MovieSceneScriptingDoubleChannel, default=0, keys [[3680,0],[3850,0]]}
       Scale.X {type=MovieSceneScriptingDoubleChannel, default=1, const [1,2]}
       Scale.Y {type=MovieSceneScriptingDoubleChannel, default=1, const [1,2]}
       Scale.Z {type=MovieSceneScriptingDoubleChannel, default=1, const [1,2]}
      ]
     }]
   }
   {type=MovieScene3DAttachTrack, sections [{type=MovieScene3DAttachSection, end_frame=15518, start_frame=3600}]}
  ]
 }
]
display_rate {denominator=1, numerator=60}
tick_resolution {denominator=1, numerator=24000}
tracks []
