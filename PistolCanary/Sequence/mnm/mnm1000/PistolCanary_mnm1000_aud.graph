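# Audio Level Sequence for the PistolCanary mnm1000 cinematic, timing whooshes for character intros.
# It has no bindings or audio tracks, only one event track with a trigger section at a 60 fps display rate.
# Six trigger keys fall in three closely spaced pairs along the timeline.
# The event endpoints are not visible here, so what each key calls or plays is not shown.
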LevelSequence /PistolCanary/Sequence/mnm/mnm1000/PistolCanary_mnm1000_aud
bindings []
display_rate {denominator=1, numerator=60}
tick_resolution {denominator=1, numerator=24000}
tracks [{
  type=MovieSceneEventTrack
  display_name=Whooshes For Character Intros
  sections [{
    type=MovieSceneEventTriggerSection
    channels [None {
      type=MovieSceneScriptingEventChannel
      keys [
       [6815.22, "<Struct 'MovieSceneEvent' (0x852ff013de0) {}>"]
       [7161.495, "<Struct 'MovieSceneEvent' (0x852ff013d40) {}>"]
       [9958.7775, "<Struct 'MovieSceneEvent' (0x852ff013ca0) {}>"]
       [10433.2925, "<Struct 'MovieSceneEvent' (0x852ff013c00) {}>"]
       [14726.635, "<Struct 'MovieSceneEvent' (0x852ff013b60) {}>"]
       [15291.1725, "<Struct 'MovieSceneEvent' (0x852ff013ac0) {}>"]
      ]
     }]
   }]
 }]
