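# Camera Level Sequence for PistolCanary shot mnm1000, with three separate camera cut sections, each bracketed by brief fade pulses, and a single event trigger.
# Four CineCameraActors carry transform tracks with extra short-range sections, plus focal length, focus distance and aperture tracks on their camera components.
# CineCameraActor_2 makes a keyed move, CineCameraActor_1 and _3 hold constant base poses, and TestFraming steps through several keyed framings with changing focal length.
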LevelSequence /PistolCanary/Sequence/mnm/mnm1000/PistolCanary_mnm1000_cam
bindings [
 CineCameraActor_2 {
  id=#1
  possessed_class=CineCameraActor
  tracks [
   {
    type=MovieScene3DTransformTrack
    sections [
     {
      type=MovieScene3DTransformSection
      channels [
       Location.X {type=MovieSceneScriptingDoubleChannel, default=14267.6777, keys [[10053,3687.9521],[10386,3441.094]]}
       Location.Y {type=MovieSceneScriptingDoubleChannel, default=77171.0312, keys [[10053,65727.7578],[10386,66419.7031]]}
       Location.Z {type=MovieSceneScriptingDoubleChannel, default=8188, keys [[10053,9544.387],[10386,9265.5188]]}
       Rotation.X {type=MovieSceneScriptingDoubleChannel, default=0, const [0,2]}
       Rotation.Y {type=MovieSceneScriptingDoubleChannel, default=0, keys [[10053,-2.5631],[10386,-0.1631]]}
       Rotation.Z {type=MovieSceneScriptingDoubleChannel, default=-89.4839, keys [[10053,22.1867],[10386,20.5867]]}
       Scale.X {type=MovieSceneScriptingDoubleChannel, default=1, const [1,2]}
       Scale.Y {type=MovieSceneScriptingDoubleChannel, default=1, const [1,2]}
       Scale.Z {type=MovieSceneScriptingDoubleChannel, default=1, const [1,2]}
      ]
     }
     {
      type=MovieScene3DTransformSection
      end_frame=10386
      start_frame=10117
      channels [
       Location.X {type=MovieSceneScriptingDoubleChannel, default=0}
       Location.Y {type=MovieSceneScriptingDoubleChannel, default=0}
       Location.Z {type=MovieSceneScriptingDoubleChannel, default=0}
       Scale.X {type=MovieSceneScriptingDoubleChannel, default=0}
       Scale.Y {type=MovieSceneScriptingDoubleChannel, default=0}
       Scale.Z {type=MovieSceneScriptingDoubleChannel, default=0}
      ]
     }
     {
      type=MovieScene3DTransformSection
      end_frame=10153
      start_frame=10117
      channels [
       Location.X {type=MovieSceneScriptingDoubleChannel, default=3862.6299}
       Location.Y {type=MovieSceneScriptingDoubleChannel, default=65250.6954}
       Location.Z {type=MovieSceneScriptingDoubleChannel, default=9542.9884}
       Scale.X {type=MovieSceneScriptingDoubleChannel, default=1}
       Scale.Y {type=MovieSceneScriptingDoubleChannel, default=1}
       Scale.Z {type=MovieSceneScriptingDoubleChannel, default=1}
      ]
     }
    ]
   }
   {type=MovieSceneSpawnTrack, sections [{type=MovieSceneSpawnSection, channels [None {type=MovieSceneScriptingBoolChannel, default=true}]}]}
  ]
 }
 CineCameraActor_1 {
  id=#2
  possessed_class=CineCameraActor
  tracks [
   {
    type=MovieScene3DTransformTrack
    sections [
     {
      type=MovieScene3DTransformSection
      channels [
       Location.X {type=MovieSceneScriptingDoubleChannel, default=11362, const [11407.6621,1]}
       Location.Y {type=MovieSceneScriptingDoubleChannel, default=71249, const [70186.2578,1]}
       Location.Z {type=MovieSceneScriptingDoubleChannel, default=4317, const [4220.8765,1]}
       Rotation.X {type=MovieSceneScriptingDoubleChannel, default=0, const [-0.8454,1]}
       Rotation.Y {type=MovieSceneScriptingDoubleChannel, default=56, const [47.259,1]}
       Rotation.Z {type=MovieSceneScriptingDoubleChannel, default=-17, const [-0.6384,1]}
      ]
     }
     {
      type=MovieScene3DTransformSection
      end_frame=7165
      start_frame=6903
      channels [
       Location.X {type=MovieSceneScriptingDoubleChannel, default=0}
       Location.Y {type=MovieSceneScriptingDoubleChannel, default=0}
       Location.Z {type=MovieSceneScriptingDoubleChannel, default=0}
       Scale.X {type=MovieSceneScriptingDoubleChannel, default=0}
       Scale.Y {type=MovieSceneScriptingDoubleChannel, default=0}
       Scale.Z {type=MovieSceneScriptingDoubleChannel, default=0}
      ]
     }
     {
      type=MovieScene3DTransformSection
      end_frame=6941
      start_frame=6903
      channels [
       Location.X {type=MovieSceneScriptingDoubleChannel, default=7938.5147}
       Location.Y {type=MovieSceneScriptingDoubleChannel, default=70379.5322}
       Location.Z {type=MovieSceneScriptingDoubleChannel, default=8266.4528}
       Scale.X {type=MovieSceneScriptingDoubleChannel, default=1}
       Scale.Y {type=MovieSceneScriptingDoubleChannel, default=1}
       Scale.Z {type=MovieSceneScriptingDoubleChannel, default=1}
      ]
     }
    ]
   }
   {type=MovieSceneSpawnTrack, sections [{type=MovieSceneSpawnSection, channels [None {type=MovieSceneScriptingBoolChannel, default=true}]}]}
  ]
 }
 CameraComponent {
  id=#3
  parent_id=#2
  possessed_class=CineCameraComponent
  tracks [
   {
    type=MovieSceneFloatTrack
    display_name=Current Focal Length
    sections [{type=MovieSceneFloatSection, channels [None {type=MovieSceneScriptingFloatChannel, default=8, const [8,1]}]}]
   }
   {
    type=MovieSceneFloatTrack
    display_name="Manual Focus Distance (Focus Settings)"
    sections [{type=MovieSceneFloatSection, channels [None {type=MovieSceneScriptingFloatChannel, default=12558.5391}]}]
   }
   {
    type=MovieSceneFloatTrack
    display_name=Current Aperture
    sections [{type=MovieSceneFloatSection, channels [None {type=MovieSceneScriptingFloatChannel, default=5.6}]}]
   }
  ]
 }
 CameraComponent {
  id=#4
  parent_id=#5
  possessed_class=CineCameraComponent
  tracks [
   {
    type=MovieSceneFloatTrack
    display_name=Current Focal Length
    sections [{type=MovieSceneFloatSection, channels [None {type=MovieSceneScriptingFloatChannel, default=16}]}]
   }
   {
    type=MovieSceneFloatTrack
    display_name="Manual Focus Distance (Focus Settings)"
    sections [{type=MovieSceneFloatSection, channels [None {type=MovieSceneScriptingFloatChannel, default=12558.5391}]}]
   }
   {
    type=MovieSceneFloatTrack
    display_name=Current Aperture
    sections [{type=MovieSceneFloatSection, channels [None {type=MovieSceneScriptingFloatChannel, default=5.6}]}]
   }
  ]
 }
 CameraComponent {
  id=#6
  parent_id=#7
  possessed_class=CineCameraComponent
  tracks [
   {
    type=MovieSceneFloatTrack
    display_name=Current Focal Length
    sections [{type=MovieSceneFloatSection, channels [None {type=MovieSceneScriptingFloatChannel, default=16, keys [[6522,14],[6622,14],[6722,8]]}]}]
   }
   {
    type=MovieSceneFloatTrack
    display_name="Manual Focus Distance (Focus Settings)"
    sections [{type=MovieSceneFloatSection, channels [None {type=MovieSceneScriptingFloatChannel, default=12558.5391}]}]
   }
   {
    type=MovieSceneFloatTrack
    display_name=Current Aperture
    sections [{type=MovieSceneFloatSection, channels [None {type=MovieSceneScriptingFloatChannel, default=5.6}]}]
   }
  ]
 }
 CameraComponent {
  id=#8
  parent_id=#1
  possessed_class=CineCameraComponent
  tracks [
   {
    type=MovieSceneFloatTrack
    display_name=Current Focal Length
    sections [{type=MovieSceneFloatSection, channels [None {type=MovieSceneScriptingFloatChannel, default=18, const [18,1]}]}]
   }
   {
    type=MovieSceneFloatTrack
    display_name="Manual Focus Distance (Focus Settings)"
    sections [{type=MovieSceneFloatSection, channels [None {type=MovieSceneScriptingFloatChannel, default=12558.5391}]}]
   }
   {
    type=MovieSceneFloatTrack
    display_name=Current Aperture
    sections [{type=MovieSceneFloatSection, channels [None {type=MovieSceneScriptingFloatChannel, default=5.6}]}]
   }
  ]
 }
 CineCameraActor_3 {
  id=#5
  possessed_class=CineCameraActor
  tracks [
   {
    type=MovieScene3DTransformTrack
    sections [
     {
      type=MovieScene3DTransformSection
      channels [
       Location.X {type=MovieSceneScriptingDoubleChannel, default=10690.0293, const [17814.793,1]}
       Location.Y {type=MovieSceneScriptingDoubleChannel, default=82140.0462, const [85962.6094,1]}
       Location.Z {type=MovieSceneScriptingDoubleChannel, default=3489.6597, const [4114.5532,1]}
       Rotation.X {type=MovieSceneScriptingDoubleChannel, default=0, const [0,1]}
       Rotation.Y {type=MovieSceneScriptingDoubleChannel, default=18.8001, const [13.6001,1]}
       Rotation.Z {type=MovieSceneScriptingDoubleChannel, default=-74.0839, const [-100.2838,1]}
       Scale.X {type=MovieSceneScriptingDoubleChannel, default=1, const [1,1]}
       Scale.Y {type=MovieSceneScriptingDoubleChannel, default=1, const [1,1]}
       Scale.Z {type=MovieSceneScriptingDoubleChannel, default=1, const [1,1]}
      ]
     }
     {
      type=MovieScene3DTransformSection
      end_frame=14891
      start_frame=14849
      channels [
       Location.X {type=MovieSceneScriptingDoubleChannel, default=0}
       Location.Y {type=MovieSceneScriptingDoubleChannel, default=0}
       Location.Z {type=MovieSceneScriptingDoubleChannel, default=0}
       Scale.X {type=MovieSceneScriptingDoubleChannel, default=0}
       Scale.Y {type=MovieSceneScriptingDoubleChannel, default=0}
       Scale.Z {type=MovieSceneScriptingDoubleChannel, default=0}
      ]
     }
     {
      type=MovieScene3DTransformSection
      end_frame=15204
      start_frame=14849
      channels [
       Location.X {type=MovieSceneScriptingDoubleChannel, default=0}
       Location.Y {type=MovieSceneScriptingDoubleChannel, default=0}
       Location.Z {type=MovieSceneScriptingDoubleChannel, default=0}
       Scale.X {type=MovieSceneScriptingDoubleChannel, default=0}
       Scale.Y {type=MovieSceneScriptingDoubleChannel, default=0}
       Scale.Z {type=MovieSceneScriptingDoubleChannel, default=0}
      ]
     }
    ]
   }
   {type=MovieSceneSpawnTrack, sections [{type=MovieSceneSpawnSection, channels [None {type=MovieSceneScriptingBoolChannel, default=true}]}]}
  ]
 }
 TestFraming {
  id=#7
  possessed_class=CineCameraActor
  tracks [
   {
    type=MovieScene3DTransformTrack
    sections [
     {
      type=MovieScene3DTransformSection
      channels [
       Location.X {type=MovieSceneScriptingDoubleChannel, default=10690.0293, keys [[6522,2537.0803],[6622,7267.7812],[6722,13892.46],[14667,17814.793]]}
       Location.Y {type=MovieSceneScriptingDoubleChannel, default=82140.0462, keys [[6522,65052.5469],[6622,77082.9219],[6722,64373.7852],[14667,85962.6094]]}
       Location.Z {type=MovieSceneScriptingDoubleChannel, default=3489.6597, keys [[6522,9737.9043],[6622,14746.9746],[6722,15181.0186],[14667,4114.5532]]}
       Rotation.X {type=MovieSceneScriptingDoubleChannel, default=0, keys [[6522,0],[6622,0],[6722,-0.8455],[14667,0]]}
       Rotation.Y {type=MovieSceneScriptingDoubleChannel, default=18.8001, keys [[6522,-8.9999],[6622,-32.7999],[6722,-52.941],[14667,13.6001]]}
       Rotation.Z {type=MovieSceneScriptingDoubleChannel, default=-74.0839, keys [[6522,21.9161],[6622,-46.4839],[6722,85.7615],[14667,-100.2838]]}
       Scale.X {type=MovieSceneScriptingDoubleChannel, default=1, const [1,1]}
       Scale.Y {type=MovieSceneScriptingDoubleChannel, default=1, const [1,1]}
       Scale.Z {type=MovieSceneScriptingDoubleChannel, default=1, const [1,1]}
      ]
     }
     {
      type=MovieScene3DTransformSection
      end_frame=14891
      start_frame=14849
      channels [
       Location.X {type=MovieSceneScriptingDoubleChannel, default=0}
       Location.Y {type=MovieSceneScriptingDoubleChannel, default=0}
       Location.Z {type=MovieSceneScriptingDoubleChannel, default=0}
       Scale.X {type=MovieSceneScriptingDoubleChannel, default=0}
       Scale.Y {type=MovieSceneScriptingDoubleChannel, default=0}
       Scale.Z {type=MovieSceneScriptingDoubleChannel, default=0}
      ]
     }
     {
      type=MovieScene3DTransformSection
      end_frame=15204
      start_frame=14849
      channels [
       Location.X {type=MovieSceneScriptingDoubleChannel, default=0}
       Location.Y {type=MovieSceneScriptingDoubleChannel, default=0}
       Location.Z {type=MovieSceneScriptingDoubleChannel, default=0}
       Scale.X {type=MovieSceneScriptingDoubleChannel, default=0}
       Scale.Y {type=MovieSceneScriptingDoubleChannel, default=0}
       Scale.Z {type=MovieSceneScriptingDoubleChannel, default=0}
      ]
     }
    ]
   }
   {type=MovieSceneSpawnTrack, sections [{type=MovieSceneSpawnSection, channels [None {type=MovieSceneScriptingBoolChannel, default=true}]}]}
  ]
 }
]
display_rate {denominator=1, numerator=60}
tick_resolution {denominator=1, numerator=24000}
tracks [
 {
  type=MovieSceneEventTrack
  sections [{
    type=MovieSceneEventTriggerSection
    channels [None {type=MovieSceneScriptingEventChannel, const ["<Struct 'MovieSceneEvent' (0x852ff013ac0) {}>", 1]}]
   }]
 }
 {
  type=MovieSceneFadeTrack
  sections [{
    type=MovieSceneFadeSection
    channels [None {
      type=MovieSceneScriptingFloatChannel
      default=0
      thinned_from=19
      keys [[6822,0],[6837,1],[6852,0],[7150,0],[7165,1],[7180,0],[10038,0],[10053,1],[10068,0],[10385,0],[10400,1],[10416,0],[14767,0],[14782,1],[14798,0],[15203,0],[15218,1],[15234,0]]
     }]
   }]
 }
 {
  type=MovieSceneCameraCutTrack
  sections [
   {type=MovieSceneCameraCutSection, end_frame=7165, start_frame=6837}
   {type=MovieSceneCameraCutSection, end_frame=15218, start_frame=14782}
   {type=MovieSceneCameraCutSection, end_frame=10400, start_frame=10053}
  ]
 }
]
