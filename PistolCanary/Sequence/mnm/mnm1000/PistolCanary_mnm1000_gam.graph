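# Gameplay event sequence for the mnm1000 shot of PistolCanary, timing NPC spawns and stages, super attacks, player events and a tentacle boss.
# Four master event tracks carry that timeline, while bound actors (sequence runner, structure destroyers, NPC aim actor, boss) receive their own event triggers.
# A scaled loot destroyer spawns only for a short late window to fire two events.
# A spawned focus point stays attached through the main span.
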LevelSequence /PistolCanary/Sequence/mnm/mnm1000/PistolCanary_mnm1000_gam
bindings [
 BP_PistolCanary_SequenceRunner_Section3 {
  id=#1
  possessed_class=BP_PistolCanary_SequenceRunner_Section1_C
  tracks [{
    type=MovieSceneEventTrack
    sections [{
      type=MovieSceneEventTriggerSection
      channels [None {type=MovieSceneScriptingEventChannel, const ["<Struct 'MovieSceneEvent' (0x852ff017b20) {}>", 1]}]
     }]
   }]
 }
 BP_PistolCanary_PlayerBuiltStructureDestroyer_FullPOI {
  id=#2
  possessed_class=BP_PistolCanary_PlayerBuiltStructureDestroyer_C
  tracks [{
    type=MovieSceneEventTrack
    sections [{
      type=MovieSceneEventTriggerSection
      channels [None {type=MovieSceneScriptingEventChannel, const ["<Struct 'MovieSceneEvent' (0x852ff0176c0) {}>", 1]}]
     }]
   }]
 }
 BP_PistolCanary_PlayerBuiltStructureDestroyer_Courtyard {
  id=#3
  possessed_class=BP_PistolCanary_PlayerBuiltStructureDestroyer_C
  tracks [{
    type=MovieSceneEventTrack
    sections [{
      type=MovieSceneEventTriggerSection
      channels [None {
        type=MovieSceneScriptingEventChannel
        keys [
         [4448.885, "<Struct 'MovieSceneEvent' (0x852ff0178a0) {}>"]
         [6080.9175, "<Struct 'MovieSceneEvent' (0x852ff017940) {}>"]
         [6723, "<Struct 'MovieSceneEvent' (0x852ff0179e0) {}>"]
         [7873.6125, "<Struct 'MovieSceneEvent' (0x852ff017a80) {}>"]
         [8786.875, "<Struct 'MovieSceneEvent' (0x852ff017bc0) {}>"]
         [9966, "<Struct 'MovieSceneEvent' (0x852ff017c60) {}>"]
         [13387.0025, "<Struct 'MovieSceneEvent' (0x852ff017800) {}>"]
         [14696, "<Struct 'MovieSceneEvent' (0x852ff017760) {}>"]
        ]
       }]
     }]
   }]
 }
 BP_PistolCanary_LootDestroyer {
  id=#4
  possessed_class=BP_PistolCanary_LootDestroyer_C
  tracks [
   {
    type=MovieSceneSpawnTrack
    sections [{
      type=MovieSceneSpawnSection
      channels [None {type=MovieSceneScriptingBoolChannel, default=false, keys [[14671,false],[14672,true],[15457,false]]}]
     }]
   }
   {
    type=MovieScene3DTransformTrack
    sections [{
      type=MovieScene3DTransformSection
      channels [
       Location.X {type=MovieSceneScriptingDoubleChannel, default=14276}
       Location.Y {type=MovieSceneScriptingDoubleChannel, default=73740}
       Location.Z {type=MovieSceneScriptingDoubleChannel, default=2629.6496}
       Rotation.X {type=MovieSceneScriptingDoubleChannel, default=0}
       Rotation.Y {type=MovieSceneScriptingDoubleChannel, default=0}
       Rotation.Z {type=MovieSceneScriptingDoubleChannel, default=0}
       Scale.X {type=MovieSceneScriptingDoubleChannel, default=200}
       Scale.Y {type=MovieSceneScriptingDoubleChannel, default=250}
       Scale.Z {type=MovieSceneScriptingDoubleChannel, default=200}
      ]
     }]
   }
   {
    type=MovieSceneEventTrack
    sections [{
      type=MovieSceneEventTriggerSection
      channels [None {
        type=MovieSceneScriptingEventChannel
        keys [[14692, "<Struct 'MovieSceneEvent' (0x852ff017440) {}>"], [15272, "<Struct 'MovieSceneEvent' (0x852ff0173a0) {}>"]]
       }]
     }]
   }
  ]
 }
 TentacleBoss {
  id=#5
  possessed_class=BGA_Tentacle_C
  tracks [{
    type=MovieSceneEventTrack
    sections [{
      type=MovieSceneEventTriggerSection
      channels [None {
        type=MovieSceneScriptingEventChannel
        keys [[3695, "<Struct 'MovieSceneEvent' (0x852ff0174e0) {}>"], [15500, "<Struct 'MovieSceneEvent' (0x852ff017580) {}>"]]
       }]
     }]
   }]
 }
 BGA_PistolCanary_NPCAimActor {
  id=#6
  possessed_class=BGA_PistolCanary_NPCAimActor_C
  tracks [{
    type=MovieSceneEventTrack
    sections [{
      type=MovieSceneEventTriggerSection
      channels [None {
        type=MovieSceneScriptingEventChannel
        keys [
         [3670, "<Struct 'MovieSceneEvent' (0x852ff018200) {}>"]
         [3671, "<Struct 'MovieSceneEvent' (0x852ff0182a0) {}>"]
         [15445, "<Struct 'MovieSceneEvent' (0x852ff018340) {}>"]
        ]
       }]
     }]
   }]
 }
 BP_PistolCanary_FocusPoint {
  id=#7
  possessed_class=BP_PistolCanary_FocusPoint_C
  tracks [
   {type=MovieSceneSpawnTrack, sections [{type=MovieSceneSpawnSection, channels [None {type=MovieSceneScriptingBoolChannel, default=true}]}]}
   {type=MovieScene3DAttachTrack, sections [{type=MovieScene3DAttachSection, end_frame=15517, start_frame=3600}]}
  ]
 }
]
display_rate {denominator=1, numerator=60}
tick_resolution {denominator=1, numerator=24000}
tracks [
 {
  type=MovieSceneEventTrack
  display_name="NPC Spawn && Supers"
  sections [{
    type=MovieSceneEventTriggerSection
    channels [None {
      type=MovieSceneScriptingEventChannel
      keys [
       [3670, "<Struct 'MovieSceneEvent' (0x852ff014420) {}>"]
       [6722, "<Struct 'MovieSceneEvent' (0x852ff0144c0) {}>"]
       [9938, "<Struct 'MovieSceneEvent' (0x852ff014560) {}>"]
       [14667, "<Struct 'MovieSceneEvent' (0x852ff014600) {}>"]
      ]
     }]
   }]
 }
 {
  type=MovieSceneEventTrack
  display_name=NonComicPanel Super Attacks
  sections [{
    type=MovieSceneEventTriggerSection
    channels [None {
      type=MovieSceneScriptingEventChannel
      keys [
       [4200, "<Struct 'MovieSceneEvent' (0x852ff014ce0) {}>"]
       [4800, "<Struct 'MovieSceneEvent' (0x852ff014d80) {}>"]
       [5400, "<Struct 'MovieSceneEvent' (0x852ff014ba0) {}>"]
       [6000, "<Struct 'MovieSceneEvent' (0x852ff014b00) {}>"]
       [7200, "<Struct 'MovieSceneEvent' (0x852ff014a60) {}>"]
       [7800, "<Struct 'MovieSceneEvent' (0x852ff0149c0) {}>"]
       [8400, "<Struct 'MovieSceneEvent' (0x852ff014920) {}>"]
       [9000, "<Struct 'MovieSceneEvent' (0x852ff0146a0) {}>"]
       [10500, "<Struct 'MovieSceneEvent' (0x852ff014380) {}>"]
       [11000, "<Struct 'MovieSceneEvent' (0x852ff014240) {}>"]
       [11700, "<Struct 'MovieSceneEvent' (0x852ff014100) {}>"]
       [12300, "<Struct 'MovieSceneEvent' (0x852ff014060) {}>"]
       [12900, "<Struct 'MovieSceneEvent' (0x852ff013f20) {}>"]
       [13600, "<Struct 'MovieSceneEvent' (0x852ff013e80) {}>"]
       [14200, "<Struct 'MovieSceneEvent' (0x852ff013b60) {}>"]
      ]
     }]
   }]
 }
 {
  type=MovieSceneEventTrack
  display_name=NPC Stages Set
  sections [{
    type=MovieSceneEventTriggerSection
    channels [None {
      type=MovieSceneScriptingEventChannel
      keys [
       [3758, "<Struct 'MovieSceneEvent' (0x852ff0151e0) {}>"]
       [4535, "<Struct 'MovieSceneEvent' (0x852ff015320) {}>"]
       [4651, "<Struct 'MovieSceneEvent' (0x852ff015500) {}>"]
       [4857, "<Struct 'MovieSceneEvent' (0x852ff0156e0) {}>"]
       [4949, "<Struct 'MovieSceneEvent' (0x852ff0158c0) {}>"]
       [5068, "<Struct 'MovieSceneEvent' (0x852ff015aa0) {}>"]
       [5222, "<Struct 'MovieSceneEvent' (0x852ff015c80) {}>"]
       [6041, "<Struct 'MovieSceneEvent' (0x852ff015be0) {}>"]
       [6157, "<Struct 'MovieSceneEvent' (0x852ff015640) {}>"]
       [6363, "<Struct 'MovieSceneEvent' (0x852ff015140) {}>"]
       [7162, "<Struct 'MovieSceneEvent' (0x852ff014f60) {}>"]
       [7281, "<Struct 'MovieSceneEvent' (0x852ff013ac0) {}>"]
       [7609, "<Struct 'MovieSceneEvent' (0x852ff014e20) {}>"]
       [9139, "<Struct 'MovieSceneEvent' (0x852ff015820) {}>"]
       [9388, "<Struct 'MovieSceneEvent' (0x852ff015280) {}>"]
       [9745, "<Struct 'MovieSceneEvent' (0x852ff015dc0) {}>"]
       [10086, "<Struct 'MovieSceneEvent' (0x852ff015fa0) {}>"]
       [10311, "<Struct 'MovieSceneEvent' (0x852ff016180) {}>"]
       [10634, "<Struct 'MovieSceneEvent' (0x852ff016360) {}>"]
       [12144, "<Struct 'MovieSceneEvent' (0x852ff016540) {}>"]
       [12545, "<Struct 'MovieSceneEvent' (0x852ff016720) {}>"]
       [12969, "<Struct 'MovieSceneEvent' (0x852ff016900) {}>"]
       [13256, "<Struct 'MovieSceneEvent' (0x852ff016ae0) {}>"]
       [13375, "<Struct 'MovieSceneEvent' (0x852ff016cc0) {}>"]
       [13775, "<Struct 'MovieSceneEvent' (0x852ff016c20) {}>"]
       [15288, "<Struct 'MovieSceneEvent' (0x852ff016860) {}>"]
      ]
     }]
   }]
 }
 {
  type=MovieSceneEventTrack
  display_name=Player Events
  sections [{
    type=MovieSceneEventTriggerSection
    channels [None {
      type=MovieSceneScriptingEventChannel
      keys [
       [3670, "<Struct 'MovieSceneEvent' (0x852ff0162c0) {}>"]
       [3672, "<Struct 'MovieSceneEvent' (0x852ff0164a0) {}>"]
       [3698, "<Struct 'MovieSceneEvent' (0x852ff016680) {}>"]
       [3758, "<Struct 'MovieSceneEvent' (0x852ff0160e0) {}>"]
       [6052, "<Struct 'MovieSceneEvent' (0x852ff015f00) {}>"]
       [6652, "<Struct 'MovieSceneEvent' (0x852ff016d60) {}>"]
       [6778, "<Struct 'MovieSceneEvent' (0x852ff016e00) {}>"]
       [7166, "<Struct 'MovieSceneEvent' (0x852ff016ea0) {}>"]
       [9268, "<Struct 'MovieSceneEvent' (0x852ff016f40) {}>"]
       [9868, "<Struct 'MovieSceneEvent' (0x852ff016fe0) {}>"]
       [10050, "<Struct 'MovieSceneEvent' (0x852ff017080) {}>"]
       [10398, "<Struct 'MovieSceneEvent' (0x852ff017120) {}>"]
       [13997, "<Struct 'MovieSceneEvent' (0x852ff0171c0) {}>"]
       [14597, "<Struct 'MovieSceneEvent' (0x852ff017260) {}>"]
       [15483, "<Struct 'MovieSceneEvent' (0x852ff017300) {}>"]
      ]
     }]
   }]
 }
]
