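# Lighting Level Sequence for the PistolCanary mnm1000 shot, overriding fog, time-of-day and spotlight settings.
# Visibility tracks hide three BP_JungleFog actors, including two low-hanging ones.
# A float track holds the PistolCanary DaySequenceModifierVolume's component at full User Blend Weight, applying its day-sequence override.
# Another sets a constant high intensity on RectLight_pistolCanary, which is bound as a SpotLight despite its name.
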LevelSequence /PistolCanary/Sequence/mnm/mnm1000/PistolCanary_mnm1000_lgt
bindings [
 BP_JungleFog_LowHanging2 {
  id=#1
  possessed_class=BP_JungleFog_C
  tracks [{
    type=MovieSceneVisibilityTrack
    sections [{type=MovieSceneVisibilitySection, channels [None {type=MovieSceneScriptingBoolChannel, default=false}]}]
   }]
 }
 BP_JungleFog_LowHanging {
  id=#2
  possessed_class=BP_JungleFog_C
  tracks [{
    type=MovieSceneVisibilityTrack
    sections [{type=MovieSceneVisibilitySection, channels [None {type=MovieSceneScriptingBoolChannel, default=false}]}]
   }]
 }
 DaySequenceModifier {
  id=#3
  parent_id=#4
  possessed_class=DaySequenceModifierComponent
  tracks [{
    type=MovieSceneFloatTrack
    display_name=User Blend Weight
    sections [{type=MovieSceneFloatSection, channels [None {type=MovieSceneScriptingFloatChannel, default=1, const [1,2]}]}]
   }]
 }
 LightComponent0 {
  id=#5
  parent_id=#6
  possessed_class=SpotLightComponent
  tracks [{
    type=MovieSceneFloatTrack
    display_name=Intensity
    sections [{type=MovieSceneFloatSection, channels [None {type=MovieSceneScriptingFloatChannel, default=0, const [120000,1]}]}]
   }]
 }
 BP_JungleFog {
  id=#7
  possessed_class=BP_JungleFog_C
  tracks [{
    type=MovieSceneVisibilityTrack
    sections [{type=MovieSceneVisibilitySection, channels [None {type=MovieSceneScriptingBoolChannel, default=false}]}]
   }]
 }
 DaySequenceModifierVolume_PistolCanary {id=#4, possessed_class=DaySequenceModifierVolume, tracks []}
 RectLight_pistolCanary {id=#6, possessed_class=SpotLight, tracks []}
]
display_rate {denominator=1, numerator=60}
tick_resolution {denominator=1, numerator=24000}
tracks []
