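# Level Sequence for the technical layer of PistolCanary mnm1000, driving comic-book presentation and boss damage events.
# Comic effect post-process volume blend weight and the MPC ComicEffect scalar each hold a constant value.
# PistolCanary_PanelManager carries trigger tracks for hero panels, comic borders, boss damage management and a dev widget.
# These timed events leave panel, border and damage logic to Blueprint handlers.
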LevelSequence /PistolCanary/Sequence/mnm/mnm1000/PistolCanary_mnm1000_tec
bindings [
 PistolCanary_ComicEffect {
  id=#1
  possessed_class=PostProcessVolume
  tracks [{
    type=MovieSceneFloatTrack
    display_name=Blend Weight
    sections [{type=MovieSceneFloatSection, channels [None {type=MovieSceneScriptingFloatChannel, default=0, const [0,1]}]}]
   }]
 }
 PistolCanary_PanelManager {
  id=#2
  possessed_class=PistolCanary_PanelManager_C
  tracks [
   {
    type=MovieSceneEventTrack
    display_name=Hero Panels
    sections [{
      type=MovieSceneEventTriggerSection
      channels [None {
        type=MovieSceneScriptingEventChannel
        keys [
         [3857, "<Struct 'MovieSceneEvent' (0x852ff018480) {}>"]
         [6418, "<Struct 'MovieSceneEvent' (0x852ff0183e0) {}>"]
         [7321, "<Struct 'MovieSceneEvent' (0x852ff018160) {}>"]
         [7620, "<Struct 'MovieSceneEvent' (0x852ff0180c0) {}>"]
         [9301, "<Struct 'MovieSceneEvent' (0x852ff018020) {}>"]
         [9600, "<Struct 'MovieSceneEvent' (0x852ff017f80) {}>"]
         [10603, "<Struct 'MovieSceneEvent' (0x852ff017ee0) {}>"]
         [10778, "<Struct 'MovieSceneEvent' (0x852ff017e40) {}>"]
         [13975, "<Struct 'MovieSceneEvent' (0x852ff017da0) {}>"]
         [14339, "<Struct 'MovieSceneEvent' (0x852ff017d00) {}>"]
         [15434, "<Struct 'MovieSceneEvent' (0x852ff018520) {}>"]
         [15480, "<Struct 'MovieSceneEvent' (0x852ff0185c0) {}>"]
        ]
       }]
     }]
   }
   {
    type=MovieSceneEventTrack
    display_name=Dev Widget
    sections [{
      type=MovieSceneEventTriggerSection
      channels [None {type=MovieSceneScriptingEventChannel, const ["<Struct 'MovieSceneEvent' (0x852ff018b60) {}>", 1]}]
     }]
   }
   {
    type=MovieSceneEventTrack
    display_name=Boss Damage Management
    sections [{
      type=MovieSceneEventTriggerSection
      channels [None {
        type=MovieSceneScriptingEventChannel
        keys [
         [3749, "<Struct 'MovieSceneEvent' (0x852ff016e00) {}>"]
         [3752, "<Struct 'MovieSceneEvent' (0x852ff016d60) {}>"]
         [3765, "<Struct 'MovieSceneEvent' (0x852ff015f00) {}>"]
         [3780, "<Struct 'MovieSceneEvent' (0x852ff0160e0) {}>"]
         [3790, "<Struct 'MovieSceneEvent' (0x852ff016680) {}>"]
         [6903, "<Struct 'MovieSceneEvent' (0x852ff0164a0) {}>"]
         [10117, "<Struct 'MovieSceneEvent' (0x852ff0162c0) {}>"]
         [14849, "<Struct 'MovieSceneEvent' (0x852ff016860) {}>"]
         [15293, "<Struct 'MovieSceneEvent' (0x852ff016a40) {}>"]
         [15374, "<Struct 'MovieSceneEvent' (0x852ff016c20) {}>"]
        ]
       }]
     }]
   }
   {
    type=MovieSceneEventTrack
    display_name=Comic Border
    sections [{
      type=MovieSceneEventTriggerSection
      channels [None {
        type=MovieSceneScriptingEventChannel
        keys [
         [6652, "<Struct 'MovieSceneEvent' (0x852ff0165e0) {}>"]
         [7166, "<Struct 'MovieSceneEvent' (0x852ff016540) {}>"]
         [9868, "<Struct 'MovieSceneEvent' (0x852ff016400) {}>"]
         [10398, "<Struct 'MovieSceneEvent' (0x852ff016360) {}>"]
         [14597, "<Struct 'MovieSceneEvent' (0x852ff016220) {}>"]
         [15220, "<Struct 'MovieSceneEvent' (0x852ff016180) {}>"]
         [15416, "<Struct 'MovieSceneEvent' (0x852ff016040) {}>"]
        ]
       }]
     }]
   }
  ]
 }
]
display_rate {denominator=1, numerator=60}
tick_resolution {denominator=1, numerator=24000}
tracks [{
  type=MovieSceneMaterialParameterCollectionTrack
  display_name=PistolCanary_MPC
  sections [{type=MovieSceneParameterSection, channels [ComicEffect {type=MovieSceneScriptingFloatChannel, const [0,1]}]}]
 }]
