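# VFX Level Sequence for the mnm1000 sequence of PistolCanary, driving a ground tentacle portal, a large portal disc and a portal charge effect.
# The ground portal Niagara actor shifts position before its system lifecycle section begins, while the disc mesh holds a fixed transform.
# The charge component is keyed active then inactive, and its Lightning Spawn Rate parameter pulses in four bursts that each taper before dropping to zero.
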LevelSequence /PistolCanary/Sequence/mnm/mnm1000/PistolCanary_mnm1000_vfx
bindings [
 NiagaraComponent0 {
  id=#1
  parent_id=#2
  possessed_class=NiagaraComponent
  tracks [
   {
    type=MovieSceneParticleTrack
    display_name=FX System
    sections [{type=MovieSceneParticleSection, channels [None {type=MovieSceneScriptingParticleChannel, const ["<ParticleKey.ACTIVATE: 0>", 2]}]}]
   }
   {type=MovieSceneNiagaraSystemTrack, sections [{type=MovieSceneNiagaraSystemSpawnSection, end_frame=15517, start_frame=3600}]}
  ]
 }
 NS_PistolCanary_Tentacle_GroundPortal_01 {
  id=#2
  possessed_class=NiagaraActor
  tracks [
   {
    type=MovieScene3DTransformTrack
    sections [{
      type=MovieScene3DTransformSection
      channels [
       Location.X {type=MovieSceneScriptingDoubleChannel, default=14348, keys [[2972,14348],[3600,14328]]}
       Location.Y {type=MovieSceneScriptingDoubleChannel, default=69801, keys [[2972,69801],[3600,70128]]}
       Location.Z {type=MovieSceneScriptingDoubleChannel, default=4356, const [4356,1]}
       Rotation.X {type=MovieSceneScriptingDoubleChannel, default=0, const [0,1]}
       Rotation.Y {type=MovieSceneScriptingDoubleChannel, default=0, const [0,1]}
       Rotation.Z {type=MovieSceneScriptingDoubleChannel, default=0, const [0,1]}
       Scale.X {type=MovieSceneScriptingDoubleChannel, default=1.1, const [1.1,1]}
       Scale.Y {type=MovieSceneScriptingDoubleChannel, default=1.1, const [1.1,1]}
       Scale.Z {type=MovieSceneScriptingDoubleChannel, default=1.1, const [1.1,1]}
      ]
     }]
   }
   {
    type=MovieSceneVisibilityTrack
    sections [{type=MovieSceneVisibilitySection, channels [None {type=MovieSceneScriptingBoolChannel, default=true, const [true,1]}]}]
   }
  ]
 }
 SM_GiftedTraveler_PortalDisc {
  id=#3
  possessed_class=FortStaticMeshActor
  tracks [{
    type=MovieScene3DTransformTrack
    sections [{
      type=MovieScene3DTransformSection
      channels [
       Location.X {type=MovieSceneScriptingDoubleChannel, default=14348, const [14348,1]}
       Location.Y {type=MovieSceneScriptingDoubleChannel, default=69801, const [69801,1]}
       Location.Z {type=MovieSceneScriptingDoubleChannel, default=3996, const [3996,1]}
       Rotation.X {type=MovieSceneScriptingDoubleChannel, default=0, const [0,1]}
       Rotation.Y {type=MovieSceneScriptingDoubleChannel, default=0, const [0,1]}
       Rotation.Z {type=MovieSceneScriptingDoubleChannel, default=0, const [0,1]}
       Scale.X {type=MovieSceneScriptingDoubleChannel, default=40, const [40,1]}
       Scale.Y {type=MovieSceneScriptingDoubleChannel, default=40, const [40,1]}
       Scale.Z {type=MovieSceneScriptingDoubleChannel, default=40, const [40,1]}
      ]
     }]
   }]
 }
 NiagaraComponent0 {
  id=#4
  parent_id=#5
  possessed_class=NiagaraComponent
  tracks [
   {
    type=MovieSceneParticleTrack
    display_name=FX System
    sections [{
      type=MovieSceneParticleSection
      channels [None {type=MovieSceneScriptingParticleChannel, keys [[3600, "<ParticleKey.ACTIVATE: 0>"], [14708, "<ParticleKey.DEACTIVATE: 1>"]]}]
     }]
   }
   {type=MovieSceneNiagaraSystemTrack, sections [{type=MovieSceneNiagaraSystemSpawnSection, end_frame=15517, start_frame=3600}]}
   {
    type=MovieSceneNiagaraFloatParameterTrack
    display_name=Lightning Spawn Rate
    sections [{
      type=MovieSceneFloatSection
      channels [None {
        type=MovieSceneScriptingFloatChannel
        default=0
        keys [[3600,0],[3765,0],[3809,12.5],[3948,12.5],[4154,5],[4248,0],[5211,0],[5212,12.5],[5374,12.5],[5630,5],[5631,0],[6239,0],[6283,12.5],[6428,12.5],[6720,5],[6721,0],[9030,0],[9128,12.5],[9248,12.5],[9499,5],[9500,0]]
       }]
     }]
   }
  ]
 }
 NS_PistolCanary_PortalCharge {id=#5, possessed_class=NiagaraActor, tracks []}
]
display_rate {denominator=1, numerator=60}
tick_resolution {denominator=1, numerator=24000}
tracks []
